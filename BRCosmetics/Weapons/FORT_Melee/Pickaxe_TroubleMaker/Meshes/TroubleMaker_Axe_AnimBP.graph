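# Anim Blueprint for the TroubleMaker pickaxe that continuously spins its dyn_Spin bone, faster during melee animations.
# Each update eases a melee-playing weight toward bIsPlayingMeleeAnim, which blends the spin speed between idle and fast rates.
# Delta-time-scaled speed accumulates into a roll rotator, applied additively in bone space by a Modify Bone node.
# Both sides of the spin threshold branch do the same addition, so the angle never wraps.
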AnimBlueprint /BRCosmetics/Weapons/FORT_Melee/Pickaxe_TroubleMaker/Meshes/TroubleMaker_Axe_AnimBP
target_skeleton=&/TroubleMaker_Axe_Skeleton
graphs [
 AnimGraph {nodes [
   n1 = "Transform (Modify) Bone - Bone: dyn_Spin"(
    ComponentPose=Local To Component(LocalPose=Input Pose())
    Alpha=1
    Translation="0,0,0"
    Rotation=Get Top_Rotate() -> (Top_Rotate:Rotator)
    Scale="1,1,1"
   ) {
    rotationMode=BMM_Additive
    rotationSpace=BCS_BoneSpace
    translationMode=BMM_Additive
    translationSpace=BCS_BoneSpace
    alphaBoolBlend {blendOption=Linear}
    alphaScaleBiasClamp {bInterpResult=true, interpSpeedDecreasing=2, interpSpeedIncreasing=200, scale=300}
    boneToModify {boneName=dyn_Spin}
   }
   Output Pose(Result=Component To Local(ComponentPose=@n1)) {layerGroup=None}
  ]}
 EventGraph {nodes [
   n1 = Event BlueprintUpdateAnimation()
   > n2 = Sequence()
   n3 = Get bIsPlayingMeleeAnim()
   n4 = Set Is Playing Melee Anim(
    execute=@n2.then_1
    Is Playing Melee Anim:double=FInterpTo(
     Current=Get Is Playing Melee Anim() -> (Is Playing Melee Anim:double)
     Target="To Float (Boolean)"(InBool=@n3)
     DeltaTime=@n1.DeltaTimeX
     InterpSpeed=1.5
    )
   ) -> (Output_Get:double)
   n5 = float * float(A=1, B=Lerp(A=30, B=1000, Alpha=@n4))
   n6 = Branch(execute=@n2.then_0, Condition="float >= float"(A=Get Top_Spin() -> (Top_Spin:double), B=@n5))
   n7 = float + float(A=Get Top_Spin(), B=float * float(A=@n5, B=@n1.DeltaTimeX) {is=title K2Node_PromotableOperator}) {is=title K2Node_CommutativeAssociativeBinaryOperator}
   > n8 = Set Top_Spin(Top_Spin:double=@n7) -> (Output_Get:double)
   Set Top_Rotate(execute=[@n8.then, @n9.then], Top_Rotate:Rotator=MakeRotator(Roll=@n7)) -> (Output_Get:Rotator)
   n9 = Set Top_Spin(execute=@n6.else, Top_Spin=@n7)
  ]}
]
variables [
 Is Playing Melee Anim {type=number, default=0.8766}
 Top_Rotate {type=Rotator, default=<rot 0 0 63104.2344>}
 Top_Spin {type=number, default=63064.5756}
]
